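# Conversation database for the NovaCrab NPC, defining its greeting and the branches that follow.
# A single speech node has a default line plus contextual lines for a first-ever conversation or specific player outfits, each with a curve-driven priority.
# That node fans out through link nodes to shared NPC quest and Jethro services entry points, a Season35 "Found" entry and NovaCrab's own dialogue entry, all referenced by tag.
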ConversationDatabase /JethroLager/NPCs/Tandem/NovaCrab/Jethro_Conversation_NovaCrab
entry_points [{tag=Jethro.Conversation.Entry.NovaCrab, destinations [#1]}]
nodes [
 ConversationEntryPointNode_7 {
  id=#1
  class=ConversationEntryPointNode
  outputs [#2]
  properties {entryTag {tagName=Jethro.Conversation.Entry.NovaCrab}}
 }
 Speech_C_4 {
  id=#2
  class=Speech_C
  outputs [#3, #4, #5, #6]
  properties {
   nodeName=""
   generalConfig {
    defaultMessage="You don't give up, do ya?"
    contextualMessages [
     {
      message="Whatever it is you want, I'm not interested."
      contextRequirements [{requirement {class=FortControllerRequirement_IsFirstEverConversationWithNPC}}]
      priority {curve {rowName=Tandem.Speech.Priority.FirstEverInteraction}}
     }
     {
      message=What did I do now?
      contextRequirements [{requirement {
         class=FortControllerRequirement_HasCID
         properties {allowedCIds [/BRCosmetics/Athena/Items/Cosmetics/Characters/Character_DistantEchoCastle]}
        }}]
      priority {curve {rowName=Tandem.Speech.Priority.OutfitReaction}}
     }
     {
      message=I figured you'd be logging some flight hours.
      contextRequirements [{requirement {class=FortControllerRequirement_HasCID, properties {allowedCIds [/BRCosmetics/Athena/Items/Cosmetics/Characters/Character_DistantEchoPro]}}}]
      priority {curve {rowName=Tandem.Speech.Priority.OutfitReaction}}
     }
     {
      message="If you're looking for Chewie, I haven't seen him."
      contextRequirements [{requirement {
         class=FortControllerRequirement_HasCID
         properties {allowedCIds [/BRCosmetics/Athena/Items/Cosmetics/Characters/Character_HydraTrumpetCoach]}
        }}]
      priority {curve {rowName=Tandem.Speech.Priority.OutfitReaction}}
     }
    ]
   }
  }
 }
 ConversationLinkNode_3 {
  id=#6
  class=ConversationLinkNode
  properties {remoteEntryTag {tagName=Jethro.Conversation.Entry.NovaCrab.Dialogue}}
 }
 ConversationLinkNode_2 {
  id=#5
  class=ConversationLinkNode
  properties {remoteEntryTag {tagName=Conversation.Entry.Season35.Found.NovaCrab}}
 }
 ConversationLinkNode_0 {id=#4, class=ConversationLinkNode, properties {remoteEntryTag {tagName=Conversation.Entry.JethroServices}}}
 ConversationLinkNode_1 {id=#3, class=ConversationLinkNode, properties {remoteEntryTag {tagName=Conversation.Entry.NPCLibrary.Quests}}}
]
